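# Additive unlit sprite particle material for an energy glow that mimics a decal where it meets scene geometry.
# Panned smoke noise, tinted by particle color, adds to a Fresnel edge taken from flattened scene-buffer world normals, with dynamic parameters driving pan and exponent.
# Opacity combines that Fresnel, a radial mask, noise-broken depth fades and optional time-of-day scaling; low-end paths substitute constants.
# An offset term pulls vertices toward particle center near the camera.
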Material /Figment_S08/Effects/Fort_Effects/Materials/ParticleDecals/M_Energy_Glow_Particle_Decal
graph {
 properties {bUsedWithParticleSprites=true, blendMode=BLEND_Additive, pixelDepthOffsetMode=PDOM_Legacy, shadingModel=MSM_Unlit, usedShadingModels=Unlit}
 nodes [
  n1 = VectorParameter() {parameterName=DarkColor, defaultValue=<rgba 0.0501 0.0332 0.1302 1>}
  n2 = TextureCoordinate()
  n3 = AppendVector(A=ScalarParameter() {defaultValue=1, parameterName=NoiseTilingU}, B=ScalarParameter() {defaultValue=0.6, parameterName=NoiseTilingV})
  n4 = DynamicParameter() {paramNames [FresnelExponent, Refraction, PanAdd, Random]}
  n5 = Multiply(A=Time(), B=LinearInterpolate(Alpha=@n4.Random) {constA=0.4, constB=1.4})
  n6 = Add(A=@n4.PanAdd, B=@n4.PanAdd)
  n7 = ScalarParameter() {defaultValue=0.05, parameterName=PannerSpeedX}
  n8 = ScalarParameter() {defaultValue=0.5, parameterName=PannerSpeedY}
  n9 = TextureSample(UVs=Panner(Coordinate=Multiply(A=@n2, B=@n3), Time=Add(A=@n5, B=@n6), Speed=AppendVector(A=@n7, B=@n8)) {speedX=0.05, speedY=0.5}) {texture=/Game/Effects/Fort_Effects/Textures/Tile/T_Inky_Smoke_Tile_Chromatic}
  n10 = Multiply(A=Desaturation(None=@n9.RGB, Fraction=ScalarParameter() {defaultValue=-1.5, parameterName=Desaturation Factor})) {constB=1.5}
  n11 = ParticleColor()
  n12 = ScalarParameter() {defaultValue=0.6, parameterName=Lack of Particle Influence}
  n13 = SceneTexture() {sceneTextureId=PPI_WorldNormal}
  n14 = Fresnel(
   ExponentIn=@n4.FresnelExponent
   Normal=MaterialFunctionCall(Normal=ComponentMask(None=@n13.Color) {b=true, g=true, r=true}, Flatness=Constant() {r=0.2}) {materialFunction=&FlattenNormal}.Result
  ) {baseReflectFraction=0, exponent=1}
  n15 = Constant() {r=0.3}
  n16 = ShadingPathSwitch(Default=@n14, Deferred=@n14, Forward=@n15, Mobile=@n14)
  n17 = FeatureLevelSwitch(Default=@n16, ES2=@n15, ES3_1=@n15, SM4_REMOVED=@n16)
  Add(
   A=Multiply(A=Add(A=@n1.RGB, B=@n10), B=LinearInterpolate(A=@n11.RGB, Alpha=@n12) {constAlpha=0.6})
   B=Multiply(A=Multiply(A=@n11.RGB, B=@n17), B=ScalarParameter() {defaultValue=5, parameterName=WorldNormalIntensity}) {constB=5, desc=Fresnel edge color}
  )
  n18 = ComponentMask(None=WorldPosition().XYZ) {b=true}
  n19 = Add(A=Frac(None=Multiply(A=@n18) {constB=0.3})) {constB=0.2}
  Clamp(None=Add(A=Floor(None=@n19)) {constB=0.8})
  n20 = ParticlePositionWS()
  n21 = ObjectPositionWS()
  n22 = FeatureLevelSwitch(Default=@n20, ES2=@n21, ES3_1=@n21, SM4_REMOVED=@n20, SM5=@n20, SM6=@n20) {desc="Please fix particle position on mobile :("}
  n23 = Distance(A=@n22, B=CameraPositionWS())
  n24 = ScalarParameter() {defaultValue=250, parameterName=NearFadeDistance}
  LinearInterpolate(A=Subtract(A=@n22, B=WorldPosition().XYZ), Alpha=Clamp(None=Divide(A=@n23, B=@n24))) {constB=0}
  MaterialFunctionCall() {materialFunction=&DebugTimeSine}
  n25 = ScalarParameter() {parameterName=OpacityMorning}
  n26 = ScalarParameter() {parameterName=OpacityDay}
  n27 = ScalarParameter() {defaultValue=1, parameterName=OpacityEvening}
  n28 = ScalarParameter() {defaultValue=1, parameterName=OpacityNight}
  n29 = Constant() {r=0.5}
  n30 = SphereMask(A=TextureCoordinate(), B=@n29, Radius=@n29) {hardnessPercent=0}
  n31 = ScalarParameter() {defaultValue=2, parameterName=SphereMaskPower}
  n32 = DepthFade(FadeDistance=ScalarParameter() {defaultValue=44, parameterName=VisibleDepthDistance}) {fadeDistanceDefault=80}
  n33 = DepthFade(Opacity=Add(A=@n9.RGB, B=OneMinus(None=@n32)), FadeDistance=ScalarParameter() {defaultValue=20, parameterName=DepthFade}) {fadeDistanceDefault=15}
  n34 = Constant() {r=0.75}
  n35 = Multiply(
   A=Multiply(A=Multiply(A=@n11.A, B=@n17) {constA=1}, B=Power(Base=@n30, Exp=@n31))
   B=FeatureLevelSwitch(Default=@n33, ES2=@n34, ES3_1=@n34, SM4_REMOVED=@n33, SM5=@n33, SM6=@n33)
  )
  StaticSwitchParameter(
   True=Multiply(
    A=MaterialFunctionCall(Color_Morning=@n25, Color_Day=@n26, Color_Evening=@n27, Color_Night=@n28) {materialFunction=&Color_Time_of_Day}.Result
    B=@n35
   )
   False=@n35
  ) {parameterName=UseTimeofDayOnOpacity}
 ]
}
